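# Parent material for the head of the MollyKit female medium player character cosmetic, built on material attributes with shading models set from the graph.
# The graph chains shared functions: MF_CharacterParent_2023_PreFX feeds CharacterSkinFX, whose result goes into MF_CharacterParent_2023_PostFX alongside other PreFX outputs.
# Skin effects are thus injected between the two halves of the common 2023 character shader.
# Usage flags cover skeletal meshes, morph targets, mesh deformers and Nanite.
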Material /BRCosmetics/Characters/Player/Female/Medium/Bodies/F_MED_MollyKit/Materials/M_MollyKit_Head_Parent
graph {
 properties {
  bUseMaterialAttributes=true
  bUsedWithMeshDeformer=true
  bUsedWithMorphTargets=true
  bUsedWithNanite=true
  bUsedWithSkeletalMesh=true
  pixelDepthOffsetMode=PDOM_Legacy
  shadingModel=MSM_FromMaterialExpression
  usedShadingModels=Default Lit | Subsurface | Clear Coat | Subsurface Profile | Cloth | Eye
 }
 nodes [
  n1 = MaterialFunctionCall() {materialFunction=&MF_CharacterParent_2023_PreFX}
  MaterialFunctionCall(
   Material=MaterialFunctionCall(PreskinnedPosition=@n1.Pre skinned Local Position, Gmap=@n1.Pre skinned Local Position, MaterialAttributes=@n1.Pre skinned Local Position) {materialFunction=&CharacterSkinFX}.Result
   M=@n1.M
   Pre Skinned Local Position=@n1.M
   Distance Blend=@n1.M
   Fresnel=@n1.M
  ) {materialFunction=&MF_CharacterParent_2023_PostFX}
 ]
}
