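# Animation Blueprint for the F_MED_MollyKit female body cosmetic, adding deformation and secondary motion to the pose copied from its attached parent mesh.
# A Control Rig fed a smoothed emote value, a weighted belt/pelvis/spine constraint on deform_milk_MSTR and a crouch-only deform_milk_01 offset build the cached Deform Pose.
# Chained AnimDynamics nodes then simulate spanner, screwdriver and pants bones.
# The event graph also sets IsCrouchingAndNOTMoving, which the AnimGraph never reads.
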AnimBlueprint /BRCosmetics/Characters/Player/Female/Medium/Bodies/F_MED_MollyKit/Meshes/F_MED_MollyKit_AnimBP
target_skeleton=&/F_MED_MollyKit_Skeleton
graphs [
 AnimGraph {nodes [
   n1 = Anim Dynamics - Bone: dyn_spanner(ComponentPose=Local To Component(LocalPose=Use cached pose 'Deform Pose'()), Alpha=1, ExternalForce="0,0,0") {
    angularDampingOverride=1
    angularSpringConstant=1000
    bAngularSpring=true
    bOverrideAngularDamping=true
    bUseGravityOverride=true
    simulationSpace=RootRelative
    alphaBoolBlend {blendOption=Linear}
    alphaScaleBiasClamp {bClampResult=true}
    boundBone {boneName=dyn_spanner}
    componentAppliedLinearAccClamp=<1000 1000 1000>
    componentLinearAccScale=<1 1 1>
    componentLinearVelScale=<1 1 1>
    gravityOverride {z=-980}
    physicsBodyDefinitions [{
      boundBone {boneName=dyn_spanner}
      boxExtents=<15 15 10>
      constraintSetup {angularLimitsMax=<0.01 10 0.01>, angularLimitsMin=<-0.01 -2 -1.5>, angularTarget=<1 0 0>}
      localJointOffset=<-1 0 0>
     }]
   }
   n2 = Anim Dynamics - Bone: dyn_screwdriver_01(ComponentPose=@n1, Alpha=1, ExternalForce="0,0,0") {
    angularBiasOverride=1
    angularDampingOverride=1
    angularSpringConstant=15000
    bAngularSpring=true
    bOverrideAngularBias=true
    bOverrideAngularDamping=true
    bUseGravityOverride=true
    simulationSpace=Actor
    alphaBoolBlend {blendOption=Linear}
    alphaScaleBiasClamp {bClampResult=true}
    boundBone {boneName=dyn_screwdriver_01}
    componentAppliedLinearAccClamp=<500 500 500>
    componentLinearAccScale=<1 1 1>
    componentLinearVelScale=<1 1 1>
    gravityOverride {z=-980}
    physicsBodyDefinitions [{
      boundBone {boneName=dyn_screwdriver_01}
      boxExtents=<30 30 30>
      constraintSetup {angularLimitsMax=<0.1 1.5 1>, angularLimitsMin=<-0.1 -1.2 -1.5>, angularTarget=<1 0 0>}
      localJointOffset=<-1 0 0>
     }]
   }
   n3 = Anim Dynamics - Bone: dyn_screwdriver_02(ComponentPose=@n2, Alpha=1, ExternalForce="0,0,0") {
    angularBiasOverride=0.9999
    angularDampingOverride=1
    angularSpringConstant=18000
    bAngularSpring=true
    bOverrideAngularBias=true
    bOverrideAngularDamping=true
    bUseGravityOverride=true
    simulationSpace=Actor
    alphaBoolBlend {blendOption=Linear}
    alphaScaleBiasClamp {bClampResult=true}
    boundBone {boneName=dyn_screwdriver_02}
    componentAppliedLinearAccClamp=<700 700 700>
    componentLinearAccScale=<0.5 0.5 0.5>
    componentLinearVelScale=<0.5 0.5 0.5>
    gravityOverride {z=-900}
    physicsBodyDefinitions [{
      boundBone {boneName=dyn_screwdriver_02}
      boxExtents=<40 30 20>
      constraintSetup {angularLimitsMax=<0.1 1.5 1.5>, angularLimitsMin=<-0.1 -1.2 -1>, angularTarget=<1 0 0>}
      localJointOffset=<-3 0 0>
     }]
   }
   n4 = Anim Dynamics - Bone: dyn_screwdriver_03(ComponentPose=@n3, Alpha=1, ExternalForce="0,0,0") {
    angularBiasOverride=0.9999
    angularDampingOverride=1
    angularSpringConstant=20000
    bAngularSpring=true
    bOverrideAngularBias=true
    bOverrideAngularDamping=true
    bUseGravityOverride=true
    simulationSpace=Actor
    alphaBoolBlend {blendOption=Linear}
    alphaScaleBiasClamp {bClampResult=true}
    boundBone {boneName=dyn_screwdriver_03}
    componentAppliedLinearAccClamp=<500 500 500>
    componentLinearAccScale=<1 1 1>
    componentLinearVelScale=<1 1 1>
    gravityOverride {z=-980}
    physicsBodyDefinitions [{
      boundBone {boneName=dyn_screwdriver_03}
      boxExtents=<30 30 20>
      constraintSetup {angularLimitsMax=<0.1 1.5 1.5>, angularLimitsMin=<-0.1 -1.2 -1.5>, angularTarget=<1 0 0>}
      localJointOffset=<-2 0 0>
     }]
   }
   n5 = Anim Dynamics - Bone: dyn_pants_01_r!(ComponentPose=@n4, Alpha=1, GravityScale:float=1, ExternalForce="0,0,0") -> (Pose) {
    angularDampingOverride=0.9999
    angularSpringConstant=500
    bGravityOverrideInSimSpace=true
    bOverrideAngularDamping=true
    bUseGravityOverride=true
    simulationSpace=RootRelative
    alphaBoolBlend {blendOption=Linear}
    alphaScaleBiasClamp {bClampResult=true}
    boundBone {boneName=dyn_pants_01_r}
    componentAppliedLinearAccClamp=<3000 3000 3000>
    componentLinearAccScale=<0.6 0.6 0.6>
    componentLinearVelScale=<0.8 0.8 0.8>
    gravityOverride {z=-600}
    physicsBodyDefinitions [{
      sphereCollisionRadius=0
      boundBone {boneName=dyn_pants_01_r}
      boxExtents=<20 30 30>
      constraintSetup {angularLimitsMax=<5 15 15>, angularLimitsMin=<-5 -15 -10>, angularTarget=<0 0 0>, linearAxesMin=<-3 0 0>}
      localJointOffset=<5 0 0>
     }]
    relativeSpaceBone {boneName=upperarm_r}
   }
   n6 = Anim Dynamics - Bone: dyn_pants_02_r!(ComponentPose=@n5, Alpha=1, GravityScale:float=1, ExternalForce="0,0,0") -> (Pose) {
    angularDampingOverride=1
    bGravityOverrideInSimSpace=true
    bOverrideAngularDamping=true
    bUseGravityOverride=true
    simulationSpace=RootRelative
    alphaBoolBlend {blendOption=Linear}
    alphaScaleBiasClamp {bClampResult=true}
    boundBone {boneName=dyn_pants_02_r}
    componentAppliedLinearAccClamp=<3000 3000 3000>
    componentLinearAccScale=<0.5 0.5 0.5>
    componentLinearVelScale=<0.7 0.7 0.7>
    gravityOverride {z=-800}
    physicsBodyDefinitions [{
      sphereCollisionRadius=0
      boundBone {boneName=dyn_pants_02_r}
      boxExtents=<30 30 20>
      constraintSetup {angularLimitsMax=<0 5 20>, angularLimitsMin=<0 -10 -25>, angularTarget=<0 0 0>, linearAxesMin=<-5 0 0>}
      localJointOffset=<5 0 0>
     }]
    relativeSpaceBone {boneName=lowerarm_r}
   }
   n7 = Anim Dynamics - Bone: dyn_pants_01_l!(ComponentPose=@n6, Alpha=1, GravityScale:float=1, ExternalForce="0,0,0") -> (Pose) {
    angularDampingOverride=0.9999
    angularSpringConstant=500
    bGravityOverrideInSimSpace=true
    bOverrideAngularDamping=true
    bUseGravityOverride=true
    simulationSpace=RootRelative
    alphaBoolBlend {blendOption=Linear}
    boundBone {boneName=dyn_pants_01_l}
    componentAppliedLinearAccClamp=<3000 3000 3000>
    componentLinearAccScale=<0.6 0.6 0.6>
    componentLinearVelScale=<0.8 0.8 0.8>
    gravityOverride {z=-600}
    physicsBodyDefinitions [{
      sphereCollisionRadius=0
      boundBone {boneName=dyn_pants_01_l}
      boxExtents=<20 30 30>
      constraintSetup {angularLimitsMax=<5 15 15>, angularLimitsMin=<-5 -15 -10>, angularTarget=<0 0 0>, linearAxesMin=<-3 0 0>}
      localJointOffset=<-5 0 0>
     }]
    relativeSpaceBone {boneName=upperarm_r}
   }
   n8 = Anim Dynamics - Bone: dyn_pants_02_l!(ComponentPose=@n7, Alpha=1, GravityScale:float=1, ExternalForce="0,0,0") -> (Pose) {
    angularDampingOverride=1
    angularSpringConstant=1000
    bAngularSpring=true
    bGravityOverrideInSimSpace=true
    bOverrideAngularDamping=true
    bUseGravityOverride=true
    simulationSpace=RootRelative
    alphaBoolBlend {blendOption=Linear}
    boundBone {boneName=dyn_pants_02_l}
    componentAppliedLinearAccClamp=<3000 3000 3000>
    componentLinearAccScale=<0.5 0.5 0.5>
    componentLinearVelScale=<0.7 0.7 0.7>
    gravityOverride {z=-800}
    physicsBodyDefinitions [{
      sphereCollisionRadius=0
      boundBone {boneName=dyn_pants_02_l}
      boxExtents=<30 30 20>
      constraintSetup {angularLimitsMax=<0 10 20>, angularLimitsMin=<0 -10 -25>, angularTarget=<1 0 0>, linearAxesMin=<-5 0 0>}
      localJointOffset=<-5 0 0>
     }]
    relativeSpaceBone {boneName=lowerarm_r}
   }
   Output Pose(Result=Component To Local(ComponentPose=@n8)) {layerGroup=None}
   n9 = Control Rig!(Source=Copy Pose From Mesh() {bUseAttachedParent=true}, Alpha, IsPlayingEmote:double=Get Is Playing Emote() -> (Is Playing Emote:double)) -> (Pose) {
    alphaBoolBlend {blendOption=Linear}
    controlRigAssetReference {blueprintRigClass=&/F_MED_MollyKit_CtrlRig.F_MED_MollyKit_CtrlRig_C}
   }
   n10 = "Constraint - deform_milk_MSTR (PPPPP)"!(
    ComponentPose=Local To Component(LocalPose=@n9)
    Alpha=1
    ConstraintWeights_0=0.135
    ConstraintWeights_1:float=0.408
    ConstraintWeights_2:float=0.039
    ConstraintWeights_3:float=0.405
    ConstraintWeights_4:float=0.013
   ) -> (Pose) {
    alphaBoolBlend {blendOption=Linear}
    boneToModify {boneName=deform_milk_MSTR}
    constraintSetup [
     {targetBone {boneName=deform_belt_01}}
     {targetBone {boneName=deform_belt_02}}
     {targetBone {boneName=deform_belt_03}}
     {targetBone {boneName=pelvis}}
     {targetBone {boneName=spine_01}}
    ]
    constraintWeights [1,1,1,1,1]
   }
   n11 = "Transform (Modify) Bone - Bone: deform_milk_01"!(ComponentPose=@n10, bAlphaBoolEnabled:bool=Get bIsCrouching(), Translation="0,0,0", Rotation="2,-15,7") -> (Pose) {
    alphaInputType=Bool
    rotationMode=BMM_Additive
    translationMode=BMM_Additive
    alphaBoolBlend {blendOption=Linear}
    boneToModify {boneName=deform_milk_01}
    rotation=<rot 2 -15 7>
   }
   Save cached pose 'Deform Pose'(Pose=Component To Local(ComponentPose=@n11)) {cacheName=Deform Pose}
  ]}
 EventGraph {nodes [
   n1 = Event BlueprintUpdateAnimation()
   > Set Delta Time X(Delta Time X=@n1.DeltaTimeX)
   n2 = Get bIsMoving2D()
   > Set IsCrouchingAndNOTMoving(IsCrouchingAndNOTMoving:bool=AND Boolean(A=Get bIsCrouching(), B=NOT Boolean(A=@n2))) -> (Output_Get:bool)
   n3 = Get bIsPlayingEmote()
   > Set Is Playing Emote(Is Playing Emote:double=FInterpTo(Current=Get Is Playing Emote(), Target="To Float (Boolean)"(InBool=@n3), DeltaTime=Get Delta Time X(), InterpSpeed=10)) -> (Output_Get:double)
  ]}
]
variables [
 Delta Time X {type=number}
 Is Playing Emote {type=number}
 IsCrouchingAndNOTMoving {type=boolean}
]
